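# Control Rig for the male medium OvenDrastic player body cosmetic that procedurally poses skirt, hood, belt and lace bones from body motion.
# Parent constraints blend skirt masters between thighs and pelvis, and hood bones between spine and clavicles.
# DeltaRotation functions remap swing or twist angles into weights for additive corrective transforms, multiplying weights for combination poses.
# Thigh-to-pelvis angles drive belts and laces, with extra crouch correctives; head and neck rotations set head-tilt weights.
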ControlRigBlueprint /BRCosmetics/Characters/Player/Male/Medium/Bodies/M_MED_OvenDrastic/Meshes/M_MED_OvenDrastic_CtrlRig
graphs [
 Add_1_ContainedGraph {nodes [
   n1 = RigVMFunctionEntryNode()
   Add = &RigVMFunction_MathDoubleAdd(A=@n1.A, B=@n1.B)
   RigVMFunctionReturnNode(Result=&RigVMFunction_MathDoubleAdd(A=@Add.Result, B=@n1.C).Result)
  ]}
 Maximum_2_ContainedGraph {nodes [
   n1 = RigVMFunctionEntryNode()
   RigVMFunctionReturnNode(Result=&RigVMFunction_MathDoubleMax(B=@n1.C, A=&RigVMFunction_MathDoubleMax(A=@n1.A, B=@n1.B).Result).Result)
  ]}
 Maximum_2_ContainedGraph {nodes [
   n1 = RigVMFunctionEntryNode()
   RigVMFunctionReturnNode(Result=&RigVMFunction_MathDoubleMax(B=@n1.C, A=&RigVMFunction_MathDoubleMax(A=@n1.A, B=@n1.B).Result).Result)
  ]}
 New Function_ContainedGraph {nodes [
   RigVMFunctionEntryNode()
   > RigVMFunction_Sequence = RigVMAggregateNode()
   &RigUnit_ParentConstraint(
    AdvancedSettings=(InterpolationType=Average,RotationOrderForFilter=XZY)
    Child=(Type=Bone,Name="MSTR_frontSkirt_l")
    Filter=(TranslationFilter=(bX=false,bY=false,bZ=false),RotationFilter=(bX=True,bY=True,bZ=True),ScaleFilter=(bX=false,bY=false,bZ=false))
    Parents=((Item=(Type=Bone,Name="thigh_l"),Weight=1),(Item=(Type=Bone,Name="pelvis"),Weight=0))
    Weight=1
    ExecutePin=@RigVMFunction_Sequence.A
   )
   > &RigUnit_ParentConstraint(
    AdvancedSettings=(InterpolationType=Average,RotationOrderForFilter=XZY)
    Child=(Type=Bone,Name="MSTR_sideSkirt_l")
    Filter=(TranslationFilter=(bX=false,bY=false,bZ=false),RotationFilter=(bX=True,bY=True,bZ=True),ScaleFilter=(bX=false,bY=false,bZ=false))
    Parents=((Item=(Type=Bone,Name="thigh_l"),Weight=0.75),(Item=(Type=Bone,Name="pelvis"),Weight=0.25))
    Weight=1
   )
   > &RigUnit_ParentConstraint(
    AdvancedSettings=(InterpolationType=Average,RotationOrderForFilter=XZY)
    Child=(Type=Bone,Name="MSTR_backSkirt_l")
    Filter=(TranslationFilter=(bX=false,bY=false,bZ=false),RotationFilter=(bX=True,bY=True,bZ=True),ScaleFilter=(bX=false,bY=false,bZ=false))
    Parents=((Item=(Type=Bone,Name="thigh_l"),Weight=0.25),(Item=(Type=Bone,Name="pelvis"),Weight=0.75))
    Weight=1
   )
   > &RigUnit_ParentConstraint(
    AdvancedSettings=(InterpolationType=Average,RotationOrderForFilter=XZY)
    Child=(Type=Bone,Name="MSTR_frontSkirt_r")
    Filter=(TranslationFilter=(bX=false,bY=false,bZ=false),RotationFilter=(bX=True,bY=True,bZ=True),ScaleFilter=(bX=false,bY=false,bZ=false))
    Parents=((Item=(Type=Bone,Name="thigh_r"),Weight=1),(Item=(Type=Bone,Name="pelvis"),Weight=0))
    Weight=1
   )
   > &RigUnit_ParentConstraint(
    AdvancedSettings=(InterpolationType=Average,RotationOrderForFilter=XZY)
    Child=(Type=Bone,Name="MSTR_sideSkirt_r")
    Filter=(TranslationFilter=(bX=false,bY=false,bZ=false),RotationFilter=(bX=True,bY=True,bZ=True),ScaleFilter=(bX=false,bY=false,bZ=false))
    Parents=((Item=(Type=Bone,Name="thigh_r"),Weight=0.75),(Item=(Type=Bone,Name="pelvis"),Weight=0.25))
    Weight=1
   )
   > &RigUnit_ParentConstraint(
    AdvancedSettings=(InterpolationType=Average,RotationOrderForFilter=XZY)
    Child=(Type=Bone,Name="MSTR_backSkirt_r")
    Filter=(TranslationFilter=(bX=false,bY=false,bZ=false),RotationFilter=(bX=True,bY=True,bZ=True),ScaleFilter=(bX=false,bY=false,bZ=false))
    Parents=((Item=(Type=Bone,Name="thigh_r"),Weight=0.25),(Item=(Type=Bone,Name="pelvis"),Weight=0.75))
    Weight=1
   )
   > &RigUnit_ParentConstraint(
    AdvancedSettings=(InterpolationType=Average,RotationOrderForFilter=XZY)
    Child=(Type=Bone,Name="MSTR_frontSkirt")
    Filter=(TranslationFilter=(bX=True,bY=True,bZ=True),RotationFilter=(bX=True,bY=True,bZ=True),ScaleFilter=(bX=false,bY=false,bZ=false))
    Parents=((Item=(Type=Bone,Name="MSTR_frontSkirt_l"),Weight=0.5),(Item=(Type=Bone,Name="MSTR_frontSkirt_r"),Weight=0.5))
    Weight=1
   )
   > &RigUnit_ParentConstraint(
    AdvancedSettings=(InterpolationType=Average,RotationOrderForFilter=XZY)
    Child=(Type=Bone,Name="MSTR_backSkirt")
    Parents=((Item=(Type=Bone,Name="MSTR_backSkirt_l"),Weight=0.5),(Item=(Type=Bone,Name="MSTR_backSkirt_r"),Weight=0.5))
    Weight=1
   )
   > RigVMFunction_Sequence_1 = RigVMAggregateNode()
   DeltaRotation v4_8 = RigVMFunctionReferenceNode(
    Rotation Order=ZYX
    child=(Type=Bone,Name="thigh_l")
    outputMinMax=(X=0,Y=1)
    parent=(Type=Null,Name="thigh_l_null")
    premult=1
    rotationMinMax=(X=-20,Y=-110)
    triggerAxis=Z
    twistAxis=(X=1,Y=0,Z=0)
    ExecuteContext=@RigVMFunction_Sequence_1.A
   )
   > &RigUnit_ModifyTransforms(
    ItemToModify="((Item=(Type=Bone,Name=\"dyn_frontSkirt_01_l\"),Transform=(Rotation=(X=0.2675,Y=-0.1107,Z=-0.2155,W=-0.9326),Translation=(X=2.0327,Y=5.1964,Z=5.2991),Scale3D=(X=1,Y=1,Z=1))),(Item=(Type=Bone,Name=\"dyn_sideSkirt_01_l\"),Transform=(Rotation=(X=0,Y=0,Z=0.1132,W= ...(truncated)"
    WeightMinimum=0
    Weight=@"DeltaRotation v4_8".outputWeight
   )
   > DeltaRotation v4_9 = RigVMFunctionReferenceNode(
    Rotation Order=ZYX
    child=(Type=Bone,Name="thigh_r")
    outputMinMax=(X=0,Y=1)
    parent=(Type=Null,Name="thigh_r_null")
    premult=1
    rotationMinMax=(X=-20,Y=-110)
    triggerAxis=Z
    twistAxis=(X=1,Y=0,Z=0)
   )
   > &RigUnit_ModifyTransforms(
    ItemToModify="((Item=(Type=Bone,Name=\"dyn_frontSkirt_01_r\"),Transform=(Rotation=(X=-0.1853,Y=-0.0897,Z=-0.0439,W=0.9776),Translation=(X=1.9159,Y=-4.8032,Z=4.4859),Scale3D=(X=1,Y=1,Z=1))),(Item=(Type=Bone,Name=\"dyn_sideSkirt_01_r\"),Transform=(Rotation=(X=-0.1323,Y=0.0459 ...(truncated)"
    WeightMinimum=0
    Weight=@"DeltaRotation v4_9".outputWeight
   )
   Multiply = &RigVMFunction_MathDoubleMul(A=@"DeltaRotation v4_9".outputWeight, B=@"DeltaRotation v4_8".outputWeight)
   > &RigUnit_ModifyTransforms(
    ItemToModify="((Item=(Type=Bone,Name=\"dyn_frontSkirt_01_r\"),Transform=(Rotation=(X=0.0027,Y=-0.0303,Z=0.0871,W=0.9957),Translation=(X=0,Y=0,Z=0),Scale3D=(X=1,Y=1,Z=1))),(Item=(Type=Bone,Name=\"dyn_frontSkirt_01_l\"),Transform=(Rotation=(X=-0.0489,Y=0.0275,Z=-0.0882,W=0.99 ...(truncated)"
    WeightMinimum=0
    Weight=@Multiply.Result
   )
   > DeltaRotation v4_6 = RigVMFunctionReferenceNode(
    Rotation Order=ZYX
    child=(Type=Bone,Name="thigh_l")
    debugRotation=false
    outputMinMax=(X=0,Y=1)
    parent=(Type=Null,Name="thigh_l_null")
    premult=1
    rotationMinMax=(X=0,Y=-20)
    triggerAxis=Z
    twistAxis=(X=1,Y=0,Z=0)
   )
   > &RigUnit_ModifyTransforms(
    ItemToModify="((Item=(Type=Bone,Name=\"dyn_frontSkirt_01_l\"),Transform=(Rotation=(X=0,Y=0,Z=0,W=1),Translation=(X=0,Y=0,Z=0),Scale3D=(X=1,Y=1,Z=1))),(Item=(Type=Bone,Name=\"dyn_sideSkirt_01_l\"),Transform=(Rotation=(X=0,Y=0,Z=0,W=1),Translation=(X=0,Y=0,Z=0),Scale3D=(X=1,Y ...(truncated)"
    WeightMinimum=0
    Weight=@"DeltaRotation v4_6".outputWeight
   )
   > DeltaRotation v4_7 = RigVMFunctionReferenceNode(
    Rotation Order=ZYX
    child=(Type=Bone,Name="thigh_r")
    outputMinMax=(X=0,Y=1)
    parent=(Type=Null,Name="thigh_r_null")
    premult=1
    rotationMinMax=(X=0,Y=-20)
    triggerAxis=Z
    twistAxis=(X=1,Y=0,Z=0)
   )
   > &RigUnit_ModifyTransforms(
    ItemToModify="((Item=(Type=Bone,Name=\"dyn_frontSkirt_01_r\"),Transform=(Rotation=(X=0,Y=0,Z=0,W=1),Translation=(X=0,Y=0,Z=0),Scale3D=(X=1,Y=1,Z=1))),(Item=(Type=Bone,Name=\"dyn_sideSkirt_01_r\"),Transform=(Rotation=(X=0,Y=0,Z=0,W=1),Translation=(X=0,Y=0,Z=0),Scale3D=(X=1,Y ...(truncated)"
    WeightMinimum=0
    Weight=@"DeltaRotation v4_7".outputWeight
   )
   > &RigUnit_ModifyTransforms(
    ItemToModify=((Item=(Type=Bone,Name="dyn_frontSkirt_01"),Transform=(Rotation=(X=0,Y=0,Z=0,W=1),Translation=(X=0,Y=0,Z=0),Scale3D=(X=1,Y=1,Z=1))))
    WeightMinimum=0
    Weight=&RigVMFunction_MathDoubleMul(B=@"DeltaRotation v4_7".outputWeight, A=@"DeltaRotation v4_6".outputWeight).Result
   )
   DeltaRotation v4_1 = RigVMFunctionReferenceNode(
    Rotation Order=ZYX
    child=(Type=Bone,Name="thigh_l")
    outputMinMax=(X=0,Y=1)
    parent=(Type=Null,Name="thigh_l_null")
    premult=1
    rotationMinMax=(X=0,Y=75)
    triggerAxis=Z
    twistAxis=(X=1,Y=0,Z=0)
    ExecuteContext=@RigVMFunction_Sequence_1.B
   )
   > &RigUnit_ModifyTransforms(
    ItemToModify="((Item=(Type=Bone,Name=\"dyn_backSkirt_02\"),Transform=(Rotation=(X=0,Y=-0.3461,Z=0,W=0.9382),Translation=(X=0,Y=0,Z=0),Scale3D=(X=1,Y=1,Z=1))),(Item=(Type=Bone,Name=\"dyn_backSkirt_01_l\"),Transform=(Rotation=(X=0.1678,Y=-0.5405,Z=-0.2489,W=0.786),Translation ...(truncated)"
    WeightMinimum=0
    Weight=@"DeltaRotation v4_1".outputWeight
   )
   > DeltaRotation v4_10 = RigVMFunctionReferenceNode(
    Rotation Order=ZYX
    child=(Type=Bone,Name="thigh_r")
    debugRotation=false
    outputMinMax=(X=0,Y=1)
    parent=(Type=Null,Name="thigh_r_null")
    premult=1
    rotationMinMax=(X=0,Y=75)
    triggerAxis=Z
    twistAxis=(X=1,Y=0,Z=0)
   )
   > &RigUnit_ModifyTransforms(
    ItemToModify="((Item=(Type=Bone,Name=\"dyn_backSkirt_02\"),Transform=(Rotation=(X=0,Y=-0.3461,Z=0,W=0.9382),Translation=(X=0,Y=0,Z=0),Scale3D=(X=1,Y=1,Z=1))),(Item=(Type=Bone,Name=\"dyn_backSkirt_01_r\"),Transform=(Rotation=(X=-0.1678,Y=-0.5405,Z=0.2489,W=0.786),Translation ...(truncated)"
    WeightMinimum=0
    Weight=@"DeltaRotation v4_10".outputWeight
   )
   > &RigUnit_ModifyTransforms(
    ItemToModify="((Item=(Type=Bone,Name=\"dyn_backSkirt_02\"),Transform=(Rotation=(X=0.0616,Y=0.7044,Z=-0.0616,W=0.7044),Translation=(X=0,Y=0,Z=0),Scale3D=(X=1,Y=1,Z=1))),(Item=(Type=Bone,Name=\"dyn_frontSkirt_01\"),Transform=(Rotation=(X=0,Y=-0.6428,Z=0,W=0.766),Translation=( ...(truncated)"
    WeightMinimum=0
    Weight=&RigVMFunction_MathDoubleMul(A=@"DeltaRotation v4_10".outputWeight, B=@"DeltaRotation v4_1".outputWeight).Result
   )
   > DeltaRotation v4_11 = RigVMFunctionReferenceNode(
    Rotation Order=ZYX
    child=(Type=Bone,Name="thigh_l")
    outputMinMax=(X=0,Y=1)
    parent=(Type=Null,Name="thigh_l_null")
    premult=1
    rotationMinMax=(X=0,Y=12)
    triggerAxis=Z
    twistAxis=(X=1,Y=0,Z=0)
   )
   > &RigUnit_ModifyTransforms(
    ItemToModify="((Item=(Type=Bone,Name=\"dyn_backSkirt_01_l\"),Transform=(Rotation=(X=0,Y=0,Z=0.0523,W=0.9986),Translation=(X=0,Y=0,Z=0),Scale3D=(X=1,Y=1,Z=1))),(Item=(Type=Bone,Name=\"dyn_sideSkirt_01_l\"),Transform=(Rotation=(X=0,Y=0,Z=0,W=1),Translation=(X=0,Y=0,Z=0),Scale ...(truncated)"
    WeightMinimum=0
    Weight=@"DeltaRotation v4_11".outputWeight
   )
   > DeltaRotation v4_12 = RigVMFunctionReferenceNode(
    Rotation Order=ZYX
    child=(Type=Bone,Name="thigh_r")
    outputMinMax=(X=0,Y=1)
    parent=(Type=Null,Name="thigh_r_null")
    premult=1
    rotationMinMax=(X=0,Y=12)
    triggerAxis=Z
    twistAxis=(X=1,Y=0,Z=0)
   )
   > &RigUnit_ModifyTransforms(
    ItemToModify="((Item=(Type=Bone,Name=\"dyn_backSkirt_01_r\"),Transform=(Rotation=(X=0,Y=0,Z=0,W=1),Translation=(X=0,Y=0,Z=0),Scale3D=(X=1,Y=1,Z=1))),(Item=(Type=Bone,Name=\"dyn_sideSkirt_01_r\"),Transform=(Rotation=(X=0,Y=0,Z=0,W=1),Translation=(X=0,Y=0,Z=0),Scale3D=(X=1,Y= ...(truncated)"
    WeightMinimum=0
    Weight=@"DeltaRotation v4_12".outputWeight
   )
   > &RigUnit_ParentConstraint(
    AdvancedSettings=(InterpolationType=Average,RotationOrderForFilter=XZY)
    Child=(Type=Bone,Name="dyn_backSkirt_01")
    Parents=((Item=(Type=Bone,Name="dyn_backSkirt_01_l"),Weight=0.5),(Item=(Type=Bone,Name="dyn_backSkirt_01_r"),Weight=0.5))
    Weight=1
   )
   DeltaRotation v4_2 = RigVMFunctionReferenceNode(
    Rotation Order=ZYX
    child=(Type=Bone,Name="thigh_l")
    outputMinMax=(X=0,Y=1)
    parent=(Type=Null,Name="thigh_l_null")
    premult=1
    rotationMinMax=(X=-20,Y=-90)
    triggerAxis=Y
    twistAxis=(X=1,Y=0,Z=0)
    ExecuteContext=@RigVMFunction_Sequence_1.C
   )
   > &RigUnit_ModifyTransforms(
    ItemToModify="((Item=(Type=Bone,Name=\"dyn_frontSkirt_01_l\"),Transform=(Rotation=(X=0.0118,Y=0.0455,Z=-0.1082,W=-0.993),Translation=(X=2.8715,Y=-3.979,Z=0.1357),Scale3D=(X=1,Y=1,Z=1))),(Item=(Type=Bone,Name=\"dyn_sideSkirt_01_l\"),Transform=(Rotation=(X=0.1039,Y=0.3808,Z=0 ...(truncated)"
    WeightMinimum=0
    Weight=@"DeltaRotation v4_2".outputWeight
   )
   > DeltaRotation v4_3 = RigVMFunctionReferenceNode(
    Rotation Order=ZYX
    child=(Type=Bone,Name="thigh_r")
    outputMinMax=(X=0,Y=1)
    parent=(Type=Null,Name="thigh_r_null")
    premult=1
    rotationMinMax=(X=0,Y=-90)
    triggerAxis=Y
    twistAxis=(X=1,Y=0,Z=0)
   )
   > &RigUnit_ModifyTransforms(
    ItemToModify="((Item=(Type=Bone,Name=\"dyn_frontSkirt_01_r\"),Transform=(Rotation=(X=0.0108,Y=-0.0457,Z=-0.0783,W=0.9958),Translation=(X=1.9997,Y=0,Z=0),Scale3D=(X=1,Y=1,Z=1))),(Item=(Type=Bone,Name=\"dyn_sideSkirt_01_r\"),Transform=(Rotation=(X=-0.0948,Y=0.3004,Z=-0.0299,W ...(truncated)"
    WeightMinimum=0
    Weight=@"DeltaRotation v4_3".outputWeight
   )
   > &RigUnit_ModifyTransforms(
    ItemToModify="((Item=(Type=Bone,Name=\"dyn_backSkirt_01_l\"),Transform=(Rotation=(X=0,Y=0,Z=0,W=1),Translation=(X=0,Y=0,Z=0),Scale3D=(X=1,Y=1,Z=1))),(Item=(Type=Bone,Name=\"dyn_backSkirt_01_r\"),Transform=(Rotation=(X=0,Y=0,Z=0,W=1),Translation=(X=0,Y=0,Z=0),Scale3D=(X=1,Y= ...(truncated)"
    WeightMinimum=0
    Weight=&RigVMFunction_MathDoubleMul(A=@"DeltaRotation v4_3".outputWeight, B=@"DeltaRotation v4_2".outputWeight).Result
   )
   > DeltaRotation v4_4 = RigVMFunctionReferenceNode(
    Rotation Order=ZYX
    child=(Type=Bone,Name="thigh_l")
    outputMinMax=(X=0,Y=1)
    parent=(Type=Null,Name="thigh_l_null")
    premult=1
    rotationMinMax=(X=0,Y=30)
    triggerAxis=Y
    twistAxis=(X=1,Y=0,Z=0)
   )
   > &RigUnit_ModifyTransforms(
    ItemToModify="((Item=(Type=Bone,Name=\"dyn_frontSkirt_01_l\"),Transform=(Rotation=(X=-0.0114,Y=-0.0864,Z=-0.13,W=0.9877),Translation=(X=0,Y=0,Z=0),Scale3D=(X=1,Y=1,Z=1))),(Item=(Type=Bone,Name=\"dyn_sideSkirt_01_l\"),Transform=(Rotation=(X=0,Y=-0.1305,Z=0,W=0.9914),Translat ...(truncated)"
    WeightMinimum=0
    Weight=@"DeltaRotation v4_4".outputWeight
   )
   > DeltaRotation v4_5 = RigVMFunctionReferenceNode(
    Rotation Order=ZYX
    child=(Type=Bone,Name="thigh_r")
    outputMinMax=(X=0,Y=1)
    parent=(Type=Null,Name="thigh_r_null")
    premult=1
    rotationMinMax=(X=0,Y=30)
    triggerAxis=Y
    twistAxis=(X=1,Y=0,Z=0)
   )
   > &RigUnit_ModifyTransforms(
    ItemToModify="((Item=(Type=Bone,Name=\"dyn_frontSkirt_01_r\"),Transform=(Rotation=(X=0.0114,Y=-0.0864,Z=0.13,W=0.9877),Translation=(X=0,Y=0,Z=0),Scale3D=(X=1,Y=1,Z=1))),(Item=(Type=Bone,Name=\"dyn_sideSkirt_01_r\"),Transform=(Rotation=(X=0,Y=-0.1305,Z=0,W=0.9914),Translatio ...(truncated)"
    WeightMinimum=0
    Weight=@"DeltaRotation v4_5".outputWeight
   )
   > DeltaRotation v4_13 = RigVMFunctionReferenceNode(
    Rotation Order=ZYX
    child=(Type=Bone,Name="thigh_l")
    onlyTwist=true
    outputMinMax=(X=0,Y=1)
    parent=(Type=Null,Name="thigh_l_null")
    premult=1
    rotationMinMax=(X=0,Y=-100)
    triggerAxis=X
    twistAxis=(X=1,Y=0,Z=0)
   )
   > &RigUnit_ModifyTransforms(
    ItemToModify="((Item=(Type=Bone,Name=\"dyn_backSkirt_01\"),Transform=(Rotation=(X=0,Y=-0.0915,Z=0,W=0.9958),Translation=(X=-0.5,Y=0,Z=1.5),Scale3D=(X=1,Y=1,Z=1))),(Item=(Type=Bone,Name=\"dyn_frontSkirt_01_l\"),Transform=(Rotation=(X=0.4883,Y=-0.2274,Z=-0.0561,W=-0.8407),Tra ...(truncated)"
    WeightMinimum=0
    Weight=@"DeltaRotation v4_13".outputWeight
   )
   > DeltaRotation v4_14 = RigVMFunctionReferenceNode(
    Rotation Order=ZYX
    child=(Type=Bone,Name="thigh_r")
    onlyTwist=true
    outputMinMax=(X=0,Y=1)
    parent=(Type=Null,Name="thigh_r_null")
    premult=1
    rotationMinMax=(X=0,Y=-100)
    triggerAxis=X
    twistAxis=(X=1,Y=0,Z=0)
   )
   > &RigUnit_ModifyTransforms(
    ItemToModify="((Item=(Type=Bone,Name=\"dyn_backSkirt_01\"),Transform=(Rotation=(X=0,Y=-0.0915,Z=0,W=0.9958),Translation=(X=-0.5019,Y=-0.0003,Z=1.5093),Scale3D=(X=1,Y=1,Z=1))),(Item=(Type=Bone,Name=\"dyn_frontSkirt_01_r\"),Transform=(Rotation=(X=0.473,Y=0.0848,Z=-0.0675,W=0. ...(truncated)"
    WeightMinimum=0
    Weight=@"DeltaRotation v4_14".outputWeight
   )
   > DeltaRotation v4_15 = RigVMFunctionReferenceNode(
    Rotation Order=ZYX
    child=(Type=Bone,Name="thigh_l")
    onlyTwist=true
    outputMinMax=(X=0,Y=1)
    parent=(Type=Null,Name="thigh_l_null")
    premult=1
    rotationMinMax=(X=0,Y=45)
    triggerAxis=X
    twistAxis=(X=1,Y=0,Z=0)
   )
   > &RigUnit_ModifyTransforms(
    ItemToModify="((Item=(Type=Bone,Name=\"dyn_backSkirt_01\"),Transform=(Rotation=(X=0,Y=0,Z=0,W=1),Translation=(X=0,Y=0,Z=0),Scale3D=(X=1,Y=1,Z=1))),(Item=(Type=Bone,Name=\"dyn_frontSkirt_01_l\"),Transform=(Rotation=(X=-0.2601,Y=0.0817,Z=-0.0247,W=-0.9618),Translation=(X=0,Y= ...(truncated)"
    WeightMinimum=0
    Weight=@"DeltaRotation v4_15".outputWeight
   )
   > DeltaRotation v4_16 = RigVMFunctionReferenceNode(
    Rotation Order=ZYX
    child=(Type=Bone,Name="thigh_r")
    onlyTwist=true
    outputMinMax=(X=0,Y=1)
    parent=(Type=Null,Name="thigh_r_null")
    premult=1
    rotationMinMax=(X=0,Y=45)
    triggerAxis=X
    twistAxis=(X=1,Y=0,Z=0)
   )
   > &RigUnit_ModifyTransforms(
    ItemToModify="((Item=(Type=Bone,Name=\"dyn_frontSkirt_01_r\"),Transform=(Rotation=(X=-0.1066,Y=-0.0892,Z=-0.0096,W=0.9902),Translation=(X=0,Y=0,Z=0),Scale3D=(X=1,Y=1,Z=1))),(Item=(Type=Bone,Name=\"dyn_sideSkirt_01_r\"),Transform=(Rotation=(X=-0.1087,Y=0,Z=0,W=0.9941),Transl ...(truncated)"
    WeightMinimum=0
    Weight=@"DeltaRotation v4_16".outputWeight
   )
   > &RigUnit_ModifyTransforms(
    ItemToModify="((Item=(Type=Bone,Name=\"dyn_frontSkirt_01_l\"),Transform=(Rotation=(X=0.1064,Y=0.0097,Z=0.1839,W=-0.9771),Translation=(X=2.4484,Y=-3.076,Z=0.8712),Scale3D=(X=1,Y=1,Z=1))),(Item=(Type=Bone,Name=\"dyn_frontSkirt_01\"),Transform=(Rotation=(X=0.2625,Y=-0.2328,Z=0 ...(truncated)"
    WeightMinimum=0
    Weight=&RigVMFunction_MathDoubleMul(A=@"DeltaRotation v4_15".outputWeight, B=@"DeltaRotation v4_8".outputWeight).Result
   )
   > &RigUnit_ModifyTransforms(
    ItemToModify="((Item=(Type=Bone,Name=\"dyn_frontSkirt_01_r\"),Transform=(Rotation=(X=0.0131,Y=-0.0861,Z=0.0851,W=0.9926),Translation=(X=2.1129,Y=0.3432,Z=2.0967),Scale3D=(X=1,Y=1,Z=1))),(Item=(Type=Bone,Name=\"dyn_frontSkirt_01\"),Transform=(Rotation=(X=0,Y=-0.1169,Z=0,W=-0 ...(truncated)"
    WeightMinimum=0
    Weight=&RigVMFunction_MathDoubleMul(B=@"DeltaRotation v4_16".outputWeight, A=@"DeltaRotation v4_9".outputWeight).Result
   )
   > DeltaRotation v4_17 = RigVMFunctionReferenceNode(
    Rotation Order=ZYX
    child=(Type=Bone,Name="thigh_l")
    onlyTwist=true
    outputMinMax=(X=0,Y=1)
    parent=(Type=Null,Name="thigh_l_null")
    premult=1
    rotationMinMax=(X=0,Y=-60)
    triggerAxis=X
    twistAxis=(X=1,Y=0,Z=0)
   )
   n1 = &RigVMFunction_MathDoubleMul(A=@"DeltaRotation v4_17".outputWeight, B=@"DeltaRotation v4_8".outputWeight)
   > &RigUnit_ModifyTransforms(
    ItemToModify="((Item=(Type=Bone,Name=\"dyn_frontSkirt_01_l\"),Transform=(Rotation=(X=-0.1183,Y=0.1528,Z=0.1532,W=0.9691),Translation=(X=0.0184,Y=1.5443,Z=-0.4569),Scale3D=(X=1,Y=1,Z=1))),(Item=(Type=Bone,Name=\"dyn_frontSkirt_01\"),Transform=(Rotation=(X=-0.2579,Y=0.0836,Z= ...(truncated)"
    WeightMinimum=0
    Weight=@n1.Result
   )
   > DeltaRotation v4_18 = RigVMFunctionReferenceNode(
    Rotation Order=ZYX
    child=(Type=Bone,Name="thigh_r")
    onlyTwist=true
    outputMinMax=(X=0,Y=1)
    parent=(Type=Null,Name="thigh_r_null")
    premult=1
    rotationMinMax=(X=0,Y=-60)
    triggerAxis=X
    twistAxis=(X=1,Y=0,Z=0)
   )
   n2 = &RigVMFunction_MathDoubleMul(A=@"DeltaRotation v4_18".outputWeight, B=@"DeltaRotation v4_9".outputWeight)
   > &RigUnit_ModifyTransforms(
    ItemToModify="((Item=(Type=Bone,Name=\"dyn_frontSkirt_01_r\"),Transform=(Rotation=(X=-0.0061,Y=-0.0332,Z=0.1966,W=0.9799),Translation=(X=0.1289,Y=-3.7995,Z=-3.1695),Scale3D=(X=1,Y=1,Z=1))),(Item=(Type=Bone,Name=\"dyn_frontSkirt_01\"),Transform=(Rotation=(X=-0.1269,Y=-0.1662 ...(truncated)"
    WeightMinimum=0
    Weight=@n2.Result
   )
   > &RigUnit_ModifyTransforms(
    ItemToModify="((Item=(Type=Bone,Name=\"dyn_frontSkirt_01\"),Transform=(Rotation=(X=-0.0343,Y=-0.0026,Z=-0.0747,W=-0.9966),Translation=(X=-0.0103,Y=0.0683,Z=-1.0038),Scale3D=(X=1,Y=1,Z=1))),(Item=(Type=Bone,Name=\"dyn_frontSkirt_01_l\"),Transform=(Rotation=(X=-0.2181,Y=0.020 ...(truncated)"
    WeightMinimum=0
    Weight=&RigVMFunction_MathDoubleMul(A=@"DeltaRotation v4_17".outputWeight, B=@"DeltaRotation v4_18".outputWeight).Result
   )
   > &RigUnit_ModifyTransforms(
    ItemToModify="((Item=(Type=Bone,Name=\"dyn_frontSkirt_01\"),Transform=(Rotation=(X=-0.2361,Y=0.1268,Z=0.2473,W=-0.9311),Translation=(X=5.0999,Y=-1.4263,Z=-7.0614),Scale3D=(X=1,Y=1,Z=1))),(Item=(Type=Bone,Name=\"dyn_frontSkirt_01_l\"),Transform=(Rotation=(X=0.4084,Y=-0.1615, ...(truncated)"
    WeightMinimum=0
    Weight=&RigVMFunction_MathDoubleMul(A=@n1.Result, B=@n2.Result).Result
   )
   > &RigUnit_ModifyTransforms(
    ItemToModify="((Item=(Type=Bone,Name=\"dyn_sideSkirt_01_l\"),Transform=(Rotation=(X=0.2366,Y=0.0824,Z=-0.5267,W=-0.8123),Translation=(X=-1.3934,Y=3.1868,Z=2.7771),Scale3D=(X=1,Y=1,Z=1))),(Item=(Type=Bone,Name=\"dyn_backSkirt_01_l\"),Transform=(Rotation=(X=0.0538,Y=0.0582,Z= ...(truncated)"
    WeightMinimum=0
    Weight=&RigVMFunction_MathDoubleMul(A=@"DeltaRotation v4_13".outputWeight, B=@"DeltaRotation v4_2".outputWeight).Result
   )
   > &RigUnit_ModifyTransforms(
    ItemToModify="((Item=(Type=Bone,Name=\"dyn_sideSkirt_01_r\"),Transform=(Rotation=(X=0.2829,Y=-0.1502,Z=0.1514,W=-0.9351),Translation=(X=5.3313,Y=-2.3302,Z=4.2139),Scale3D=(X=1,Y=1,Z=1))),(Item=(Type=Bone,Name=\"dyn_backSkirt_01_r\"),Transform=(Rotation=(X=-0.0212,Y=0.032,Z= ...(truncated)"
    WeightMinimum=0
    Weight=&RigVMFunction_MathDoubleMul(A=@"DeltaRotation v4_14".outputWeight, B=@"DeltaRotation v4_3".outputWeight).Result
   )
   > &RigUnit_ModifyTransforms(
    ItemToModify="((Item=(Type=Bone,Name=\"dyn_sideSkirt_01_l\"),Transform=(Rotation=(X=-0.0473,Y=-0.0403,Z=-0.0347,W=0.9975),Translation=(X=2.3748,Y=-0.7519,Z=0.5331),Scale3D=(X=1,Y=1,Z=1))),(Item=(Type=Bone,Name=\"dyn_backSkirt_01_l\"),Transform=(Rotation=(X=0,Y=0.1077,Z=0,W= ...(truncated)"
    WeightMinimum=0
    Weight=&RigVMFunction_MathDoubleMul(B=@"DeltaRotation v4_15".outputWeight, A=@"DeltaRotation v4_2".outputWeight).Result
   )
   > &RigUnit_ModifyTransforms(
    ItemToModify="((Item=(Type=Bone,Name=\"dyn_sideSkirt_01_r\"),Transform=(Rotation=(X=0.0223,Y=-0.0415,Z=-0.0733,W=-0.9962),Translation=(X=0.0952,Y=0.0606,Z=1.1937),Scale3D=(X=1,Y=1,Z=1))),(Item=(Type=Bone,Name=\"dyn_backSkirt_01_r\"),Transform=(Rotation=(X=0,Y=0.1183,Z=0,W=- ...(truncated)"
    WeightMinimum=0
    Weight=&RigVMFunction_MathDoubleMul(B=@"DeltaRotation v4_16".outputWeight, A=@"DeltaRotation v4_3".outputWeight).Result
   )
   > &RigUnit_ModifyTransforms(
    ItemToModify="((Item=(Type=Bone,Name=\"dyn_frontSkirt_01_l\"),Transform=(Rotation=(X=0.0276,Y=0.0923,Z=-0.2556,W=0.962),Translation=(X=0,Y=0,Z=0),Scale3D=(X=1,Y=1,Z=1))),(Item=(Type=Bone,Name=\"dyn_frontSkirt_01\"),Transform=(Rotation=(X=0.493,Y=0.0501,Z=0.0495,W=0.8672),Tr ...(truncated)"
    WeightMaximum=0.5
    WeightMinimum=0
    Weight=&RigVMFunction_MathDoubleMul(B=@"DeltaRotation v4_2".outputWeight, A=@"DeltaRotation v4_8".outputWeight).Result
   )
   > &RigUnit_ModifyTransforms(
    ItemToModify="((Item=(Type=Bone,Name=\"dyn_frontSkirt_01_r\"),Transform=(Rotation=(X=0.0735,Y=-0.0811,Z=0.006,W=0.994),Translation=(X=0,Y=0,Z=0),Scale3D=(X=1,Y=1,Z=1))),(Item=(Type=Bone,Name=\"dyn_frontSkirt_01\"),Transform=(Rotation=(X=0,Y=0.2846,Z=0,W=0.9587),Translation= ...(truncated)"
    WeightMaximum=0.5
    WeightMinimum=0
    Weight=&RigVMFunction_MathDoubleMul(B=@"DeltaRotation v4_3".outputWeight, A=@"DeltaRotation v4_9".outputWeight).Result
   )
   n3 = &RigVMFunction_MathQuaternionToEuler(
    RotationOrder=ZXY
    Value=&RigUnit_GetRelativeTransformForItem(Child=(Type=Bone,Name="thigh_l"), Parent=(Type=Bone,Name="pelvis"), bChildInitial=False, bParentInitial=False).RelativeTransform.Rotation
   )
   RigVMVariableNode(
    Value=&RigVMFunction_MathDoubleRemap(SourceMaximum=-75, SourceMinimum=0, TargetMaximum=1, TargetMinimum=0, Value=@n3.Result.Z).Result
    ExecuteContext=@RigVMFunction_Sequence_1.D
   ) {variable=leg_up_l}
   > RigVMVariableNode(Value=&RigVMFunction_MathDoubleRemap(SourceMaximum=-80, SourceMinimum=-40, TargetMaximum=1, TargetMinimum=0, Value=@n3.Result.Y).Result) {variable=leg_out_l}
   > RigVMVariableNode(Value=&RigVMFunction_MathDoubleRemap(SourceMaximum=-102, SourceMinimum=-50, TargetMaximum=1, TargetMinimum=0, Value=@n3.Result.Z).Result) {variable=leg up crouching}
   > RigVMVariableNode(Value=&RigVMFunction_MathDoubleRemap(SourceMaximum=20, SourceMinimum=0, TargetMaximum=1, TargetMinimum=0, Value=@n3.Result.Z).Result) {variable=leg_back_l}
   > RigVMVariableNode(Value=&RigVMFunction_MathDoubleRemap(SourceMaximum=-80, SourceMinimum=0, TargetMaximum=1, TargetMinimum=0, Value=@n3.Result.Y).Result) {variable=leg up out crouching}
   > &RigUnit_ModifyTransforms(
    ItemToModify="((Item=(Type=Bone,Name=\"dyn_upper_belt_01_l\"),Transform=(Rotation=(X=0,Y=-0.2588,Z=0,W=0.9659),Translation=(X=0,Y=0,Z=0),Scale3D=(X=1,Y=1,Z=1))),(Item=(Type=Bone,Name=\"dyn_lace_c_01_l\"),Transform=(Rotation=(X=0.0227,Y=0.1722,Z=-0.1285,W=0.9764),Translation ...(truncated)"
    WeightMinimum=0
    Weight=RigVMVariableNode() {variable=leg_out_l}.Value
   )
   > &RigUnit_ModifyTransforms(
    ItemToModify="((Item=(Type=Bone,Name=\"dyn_upper_belt_01_l\"),Transform=(Rotation=(X=-0.0702,Y=0.171,Z=-0.1354,W=-0.9734),Translation=(X=0.0201,Y=-1.847,Z=2.0677),Scale3D=(X=1,Y=1.05,Z=1))),(Item=(Type=Bone,Name=\"dyn_lace_c_01_l\"),Transform=(Rotation=(X=0.6089,Y=0.3278,Z= ...(truncated)"
    WeightMinimum=0
    Weight=RigVMVariableNode() {variable=leg_up_l}.Value
   )
   > &RigUnit_ModifyTransforms(
    ItemToModify="((Item=(Type=Bone,Name=\"dyn_upper_belt_01_l\"),Transform=(Rotation=(X=0.1301,Y=0,Z=0,W=-0.9915),Translation=(X=0,Y=0,Z=0),Scale3D=(X=1,Y=1.038,Z=1))),(Item=(Type=Bone,Name=\"dyn_lace_c_01_l\"),Transform=(Rotation=(X=0,Y=0.0238,Z=0,W=-0.9997),Translation=(X=0, ...(truncated)"
    WeightMinimum=0
    Weight=RigVMVariableNode() {variable=leg_back_l}.Value
   )
   n4 = &RigVMFunction_MathQuaternionToEuler(
    RotationOrder=ZXY
    Value=&RigUnit_GetRelativeTransformForItem(Child=(Type=Bone,Name="thigh_r"), Parent=(Type=Bone,Name="pelvis"), bChildInitial=False, bParentInitial=False).RelativeTransform.Rotation
   )
   n5 = &RigVMFunction_MathDoubleSub(A=180, B=&RigVMFunction_MathDoubleAbs(Value=@n4.Result.Z).Result)
   > RigVMVariableNode(Value=&RigVMFunction_MathDoubleRemap(SourceMaximum=100, SourceMinimum=0, TargetMaximum=1, TargetMinimum=0, Value=@n5.Result).Result) {variable=leg_up_r}
   > RigVMVariableNode(Value=&RigVMFunction_MathDoubleRemap(SourceMaximum=100, SourceMinimum=0, TargetMaximum=1, TargetMinimum=0, Value=@n4.Result.Y).Result) {variable=leg_out_r}
   > RigVMVariableNode(Value=&RigVMFunction_MathDoubleRemap(SourceMaximum=80, SourceMinimum=0, TargetMaximum=1, TargetMinimum=0, Value=@n4.Result.Y).Result) {variable=leg up out crouching_r}
   Remap = &RigVMFunction_MathDoubleRemap(SourceMaximum=69, SourceMinimum=5, TargetMaximum=1, TargetMinimum=0, bClamp=False, Value=@n5.Result)
   > RigVMVariableNode(Value=@Remap.Result) {variable=leg up crouching_r}
   > &RigUnit_ModifyTransforms(
    ItemToModify="((Item=(Type=Bone,Name=\"dyn_lace_01_r\"),Transform=(Rotation=(X=0,Y=0.5736,Z=0,W=0.8192),Translation=(X=0,Y=0,Z=0),Scale3D=(X=1,Y=1,Z=1))),(Item=(Type=Bone,Name=\"dyn_lace_02_r\"),Transform=(Rotation=(X=0,Y=0.342,Z=0,W=0.9397),Translation=(X=0,Y=0,Z=0),Scale3 ...(truncated)"
    WeightMinimum=0
    Weight=RigVMVariableNode() {variable=leg_up_r}.Value
   )
   > n6 = &RigUnit_ModifyTransforms(
    ItemToModify="((Item=(Type=Bone,Name=\"dyn_lace_01_r\"),Transform=(Rotation=(X=0,Y=0.0436,Z=0,W=0.999),Translation=(X=0,Y=0,Z=0),Scale3D=(X=1,Y=1,Z=1))),(Item=(Type=Bone,Name=\"dyn_lace_02_r\"),Transform=(Rotation=(X=0,Y=0.1305,Z=0,W=0.9914),Translation=(X=0,Y=0,Z=0),Scale3 ...(truncated)"
    WeightMinimum=0
    Weight=RigVMVariableNode() {variable=leg_out_r}.Value
   )
   RigVMFunction_ControlFlowBranch = &RigVMFunction_ControlFlowBranch(ExecuteContext=@n6.ExecutePin, Condition=RigVMVariableNode() {variable=is crouching}.Value)
   RigVMFunctionReturnNode(ExecuteContext=@RigVMFunction_ControlFlowBranch.Completed)
   &RigUnit_ModifyTransforms(
    ItemToModify="((Item=(Type=Bone,Name=\"Base_upper_belt_l\"),Transform=(Rotation=(X=0.1913,Y=0.2535,Z=-0.1104,W=-0.9418),Translation=(X=0.3879,Y=-0.8053,Z=-4.0334),Scale3D=(X=1,Y=1.0802,Z=1))),(Item=(Type=Bone,Name=\"base_lace_c_01_l\"),Transform=(Rotation=(X=0.1567,Y=-0.012 ...(truncated)"
    WeightMinimum=0
    ExecutePin=@RigVMFunction_ControlFlowBranch.True
    Weight=RigVMVariableNode() {variable=leg up crouching}.Value
   )
   > &RigUnit_ModifyTransforms(
    ItemToModify="((Item=(Type=Bone,Name=\"base_lace_c_01_l\"),Transform=(Rotation=(X=0.1555,Y=0.1853,Z=0.0545,W=0.9688),Translation=(X=0,Y=0,Z=0),Scale3D=(X=1,Y=1,Z=1))),(Item=(Type=Bone,Name=\"base_lace_c_03_l\"),Transform=(Rotation=(X=0,Y=-0.3827,Z=0,W=0.9239),Translation=(X ...(truncated)"
    WeightMinimum=0
    Weight=RigVMVariableNode() {variable=leg up out crouching}.Value
   )
   > &RigUnit_ModifyTransforms(
    ItemToModify="((Item=(Type=Bone,Name=\"base_lace_01_r\"),Transform=(Rotation=(X=-0.4216,Y=-0.0632,Z=-0.0295,W=0.9041),Translation=(X=0,Y=0,Z=0),Scale3D=(X=1,Y=1,Z=1))),(Item=(Type=Bone,Name=\"dyn_lace_02_r\"),Transform=(Rotation=(X=0,Y=0.0698,Z=0,W=0.9976),Translation=(X=0, ...(truncated)"
    WeightMinimum=0
    Weight=RigVMVariableNode() {variable=leg up crouching_r}.Value
   )
   > &RigUnit_ModifyTransforms(
    ItemToModify="((Item=(Type=Bone,Name=\"None\"),Transform=(Rotation=(X=0,Y=0,Z=0,W=1),Translation=(X=0,Y=0,Z=0),Scale3D=(X=1,Y=1,Z=1))),(Item=(Type=Bone,Name=\"dyn_upper_belt_01_r\"),Transform=(Rotation=(X=0,Y=-0.0175,Z=0,W=0.9998),Translation=(X=-3,Y=0,Z=-2),Scale3D=(X=1,Y= ...(truncated)"
    Weight=1
    WeightMinimum=0
   )
   &RigVMFunction_MathDoubleRemap(SourceMaximum=60, SourceMinimum=40, TargetMaximum=1, TargetMinimum=0, Value=@n3.Result.Y)
   &RigUnit_SphericalPoseReader(
    ActiveRegionScaleFactors=(PositiveWidth=3.5,NegativeWidth=0.5,PositiveHeight=3,NegativeHeight=3)
    ActiveRegionSize=0.1
    Debug=(bDrawDebug=True,bDraw2D=False,bDrawLocalAxes=False,DebugScale=25,DebugSegments=20,DebugThickness=0.25)
    DriverAxis=(X=-1,Y=0,Z=0)
    DriverItem=(Type=Bone,Name="thigh_l")
    ExecutePin=()
    FalloffRegionScaleFactors=(PositiveWidth=1,NegativeWidth=0.4,PositiveHeight=1,NegativeHeight=1)
    FalloffSize=0.45
    FlipHeightScaling=False
    FlipWidthScaling=false
    OptionalParentItem=(Type=Bone,Name="None")
    RotationOffset=(X=131.8535,Y=-2.5521,Z=-11.7936)
   )
   Absolute = &RigVMFunction_MathDoubleAbs(Value=@n3.Result.Z)
   Subtract = &RigVMFunction_MathDoubleSub(A=180, B=@Absolute.Result)
  ]}
 New Function_ContainedGraph {nodes [
   RigVMFunctionEntryNode()
   > &RigUnit_ParentConstraint(
    AdvancedSettings=(InterpolationType=Average,RotationOrderForFilter=XZY)
    Child=(Type=Bone,Name="hood_back_c")
    Filter=(TranslationFilter=(bX=true,bY=true,bZ=true),RotationFilter=(bX=True,bY=True,bZ=True),ScaleFilter=(bX=True,bY=True,bZ=True))
    Parents=((Item=(Type=Bone,Name="spine_05"),Weight=0.8),(Item=(Type=Bone,Name="clavicle_l"),Weight=0.1),(Item=(Type=Bone,Name="clavicle_r"),Weight=0.1))
    Weight=1
   )
   > &RigUnit_ParentConstraint(
    AdvancedSettings=(InterpolationType=Average,RotationOrderForFilter=XZY)
    Child=(Type=Bone,Name="hood_front_l")
    Filter=(TranslationFilter=(bX=true,bY=true,bZ=true),RotationFilter=(bX=True,bY=True,bZ=True),ScaleFilter=(bX=True,bY=True,bZ=True))
    Parents=((Item=(Type=Bone,Name="spine_05"),Weight=0.7),(Item=(Type=Bone,Name="clavicle_l"),Weight=0.3))
    Weight=1
   )
   > &RigUnit_ParentConstraint(
    AdvancedSettings=(InterpolationType=Average,RotationOrderForFilter=XZY)
    Child=(Type=Bone,Name="hood_side_l")
    Filter=(TranslationFilter=(bX=true,bY=true,bZ=true),RotationFilter=(bX=True,bY=True,bZ=True),ScaleFilter=(bX=True,bY=True,bZ=True))
    Parents=((Item=(Type=Bone,Name="spine_05"),Weight=0.7),(Item=(Type=Bone,Name="clavicle_l"),Weight=0.3))
    Weight=1
   )
   > &RigUnit_ParentConstraint(
    AdvancedSettings=(InterpolationType=Average,RotationOrderForFilter=XZY)
    Child=(Type=Bone,Name="hood_back_side_l")
    Filter=(TranslationFilter=(bX=true,bY=true,bZ=true),RotationFilter=(bX=True,bY=True,bZ=True),ScaleFilter=(bX=True,bY=True,bZ=True))
    Parents=((Item=(Type=Bone,Name="spine_05"),Weight=0.75),(Item=(Type=Bone,Name="clavicle_l"),Weight=0.15))
    Weight=1
   )
   > &RigUnit_ParentConstraint(
    AdvancedSettings=(InterpolationType=Average,RotationOrderForFilter=XZY)
    Child=(Type=Bone,Name="hood_front_r")
    Filter=(TranslationFilter=(bX=true,bY=true,bZ=true),RotationFilter=(bX=True,bY=True,bZ=True),ScaleFilter=(bX=True,bY=True,bZ=True))
    Parents=((Item=(Type=Bone,Name="spine_05"),Weight=0.7),(Item=(Type=Bone,Name="clavicle_r"),Weight=0.3))
    Weight=1
   )
   > &RigUnit_ParentConstraint(
    AdvancedSettings=(InterpolationType=Average,RotationOrderForFilter=XZY)
    Child=(Type=Bone,Name="hood_side_r")
    Filter=(TranslationFilter=(bX=true,bY=true,bZ=true),RotationFilter=(bX=True,bY=True,bZ=True),ScaleFilter=(bX=True,bY=True,bZ=True))
    Parents=((Item=(Type=Bone,Name="spine_05"),Weight=0.7),(Item=(Type=Bone,Name="clavicle_r"),Weight=0.3))
    Weight=1
   )
   > &RigUnit_ParentConstraint(
    AdvancedSettings=(InterpolationType=Average,RotationOrderForFilter=XZY)
    Child=(Type=Bone,Name="hood_back_side_r")
    Filter=(TranslationFilter=(bX=true,bY=true,bZ=true),RotationFilter=(bX=True,bY=True,bZ=True),ScaleFilter=(bX=True,bY=True,bZ=True))
    Parents=((Item=(Type=Bone,Name="spine_05"),Weight=0.75),(Item=(Type=Bone,Name="clavicle_r"),Weight=0.15))
    Weight=1
   )
   > RigVMFunctionReturnNode()
  ]}
 New Function_ContainedGraph {nodes [
   RigVMFunctionEntryNode()
   RigVMFunction_MathQuaternionToEuler = &RigVMFunction_MathQuaternionToEuler(
    RotationOrder=ZYX
    Value=&RigVMFunction_MathQuaternionSwingTwist(
     TwistAxis=(X=1,Y=0,Z=0)
     Input=&RigUnit_GetRelativeTransformForItem(Child=(Type=Bone,Name="head"), Parent=(Type=Control,Name="ref_head_ctrl"), bChildInitial=False, bParentInitial=False).RelativeTransform.Rotation
    ).Swing
   )
   n1 = &RigVMFunction_MathQuaternionToEuler(
    RotationOrder=ZYX
    Value=&RigVMFunction_MathQuaternionSwingTwist(
     TwistAxis=(X=1,Y=0,Z=0)
     Input=&RigUnit_GetRelativeTransformForItem(Child=(Type=Bone,Name="neck_02"), Parent=(Type=Control,Name="ref_neck_ctrl"), bChildInitial=False, bParentInitial=False).RelativeTransform.Rotation
    ).Swing
   )
   Maximum_2 = RigVMAggregateNode(
    B=&RigVMFunction_MathDoubleRemap(SourceMaximum=120, SourceMinimum=0, TargetMaximum=1, TargetMinimum=0, Value=@RigVMFunction_MathQuaternionToEuler.Result.Y).Result
    C=&RigVMFunction_MathDoubleRemap(SourceMaximum=40, SourceMinimum=0, TargetMaximum=1, TargetMinimum=0, Value=@n1.Result.Y).Result
    A=&RigVMFunction_MathDoubleRemap(SourceMaximum=-70, SourceMinimum=0, TargetMaximum=1, TargetMinimum=0, Value=RigVMVariableNode() {variable=Variable rot l}.Value.Y).Result
   )
   > RigVMVariableNode(Value=@Maximum_2.Result) {variable=Head side l}
   n2 = &RigVMFunction_MathDoubleRemap(SourceMaximum=-70, SourceMinimum=0, TargetMaximum=1, TargetMinimum=0, Value=RigVMVariableNode() {variable=Variable rot r}.Value.Y)
   Add = RigVMAggregateNode(
    B=&RigVMFunction_MathDoubleRemap(SourceMaximum=-120, SourceMinimum=0, TargetMaximum=1, TargetMinimum=0, Value=@RigVMFunction_MathQuaternionToEuler.Result.Y).Result
    C=&RigVMFunction_MathDoubleRemap(SourceMaximum=-40, SourceMinimum=0, TargetMaximum=1, TargetMinimum=0, Value=@n1.Result.Y).Result
    A=@n2.Result
   )
   Maximum_1 = RigVMAggregateNode(B=@Add.Result, A=@n2.Result)
   > RigVMVariableNode(Value=@Maximum_1.Result) {variable=Head side r}
   RigVMFunction_AnimEasing = /Script/RigVM.RigVMFunction_AnimEasing(
    SourceMaximum=1
    SourceMinimum=0
    TargetMaximum=1
    TargetMinimum=0
    Type=CubicEaseIn
    Value=&RigVMFunction_MathDoubleRemap(SourceMaximum=-100, SourceMinimum=0, TargetMaximum=1, TargetMinimum=0, Value=@RigVMFunction_MathQuaternionToEuler.Result.Z).Result
   ) {is=struct RigVMUnitNode}
   > RigVMVariableNode(Value=&RigVMFunction_MathDoubleAdd(
     B=&RigVMFunction_MathDoubleRemap(SourceMaximum=-100, SourceMinimum=0, TargetMaximum=1, TargetMinimum=0, Value=@n1.Result.Z).Result
     A=@RigVMFunction_AnimEasing.Result
    ).Result) {variable=Head back c}
   > RigVMFunctionReturnNode()
  ]}
 New Function_ContainedGraph {nodes [
   n1 = RigVMFunctionEntryNode()
   RigVMFunction_MathBoolToInteger = &RigVMFunction_MathBoolToInteger(Value=@n1.onlyTwist)
   RigVMFunction_MathQuaternionSwingTwist = &RigVMFunction_MathQuaternionSwingTwist(
    twistAxis=@n1.twistAxis
    Input=&RigUnit_GetRelativeTransformForItem(bChildInitial=False, bParentInitial=False, child=@n1.child, parent=@n1.parent).RelativeTransform.Rotation
   )
   n2 = RigVMDispatchNode(
    Values=((),())
    Index=@RigVMFunction_MathBoolToInteger.Result
    Values.0=@RigVMFunction_MathQuaternionSwingTwist.Swing
    Values.1=@RigVMFunction_MathQuaternionSwingTwist.Twist
   ) {template="DISPATCH_RigVMDispatch_SelectInt32(in Index,in Values,out Result)"}
   RigVMFunction_MathQuaternionToEuler = &RigVMFunction_MathQuaternionToEuler(Value=@n2.Result, RotationOrder=@n1.Rotation Order)
   > DISPATCH_RigVMDispatch_Print = RigVMDispatchNode(ScreenColor=(R=1,G=1,B=1,A=1), ScreenDuration=1.5, Enabled=@n1.debugRotation, Value=@RigVMFunction_MathQuaternionToEuler.Result) {template="DISPATCH_RigVMDispatch_Print(in Prefix,in Value,in Enabled,in ScreenDuration,in ScreenColor)"}
   Cast = RigVMDispatchNode(Value=@n1.triggerAxis) {template="DISPATCH_RigVMDispatch_CastEnumToInt(in Value,out Result)"}
   DISPATCH_RigVMDispatch_SelectInt32 = RigVMDispatchNode(
    Values=(0,0,0)
    Index=@Cast.Result
    Values.0=@RigVMFunction_MathQuaternionToEuler.Result.X
    Values.1=@RigVMFunction_MathQuaternionToEuler.Result.Y
    Values.2=@RigVMFunction_MathQuaternionToEuler.Result.Z
   ) {template="DISPATCH_RigVMDispatch_SelectInt32(in Index,in Values,out Result)"}
   Remap = &RigVMFunction_MathDoubleRemap(
    bClamp=False
    Value=@DISPATCH_RigVMDispatch_SelectInt32.Result
    TargetMinimum=@n1.outputMinMax.X
    TargetMaximum=@n1.outputMinMax.Y
    SourceMinimum=@n1.rotationMinMax.X
    SourceMaximum=@n1.rotationMinMax.Y
   )
   Absolute = &RigVMFunction_MathDoubleAbs(Value=@Remap.Result)
   Multiply = &RigVMFunction_MathDoubleMul(
    A=&RigVMFunction_MathDoubleClamp(Value=RigVMDispatchNode(True=@Absolute.Result, Condition=@n1.absoluteValue, False=@Remap.Result) {template="DISPATCH_RigVMDispatch_If(in Condition,in True,in False,out Result)"}.Result).Result
    B=@n1.premult
   )
   > RigVMFunctionReturnNode(outputQuat=@n2.Result, outputWeight=@Multiply.Result)
  ]}
 New Function_ContainedGraph {nodes [
   n1 = RigVMFunctionEntryNode()
   RigVMFunction_MathBoolToInteger = &RigVMFunction_MathBoolToInteger(Value=@n1.onlyTwist)
   RigVMFunction_MathQuaternionSwingTwist = &RigVMFunction_MathQuaternionSwingTwist(
    twistAxis=@n1.twistAxis
    Input=&RigUnit_GetRelativeTransformForItem(bChildInitial=False, bParentInitial=False, child=@n1.child, parent=@n1.parent).RelativeTransform.Rotation
   )
   n2 = RigVMDispatchNode(
    Values=((),())
    Index=@RigVMFunction_MathBoolToInteger.Result
    Values.0=@RigVMFunction_MathQuaternionSwingTwist.Swing
    Values.1=@RigVMFunction_MathQuaternionSwingTwist.Twist
   ) {template="DISPATCH_RigVMDispatch_SelectInt32(in Index,in Values,out Result)"}
   RigVMFunction_MathQuaternionToEuler = &RigVMFunction_MathQuaternionToEuler(Value=@n2.Result, RotationOrder=@n1.Rotation Order)
   > DISPATCH_RigVMDispatch_Print = RigVMDispatchNode(ScreenColor=(R=1,G=1,B=1,A=1), ScreenDuration=1.5, Enabled=@n1.debugRotation, Value=@RigVMFunction_MathQuaternionToEuler.Result) {template="DISPATCH_RigVMDispatch_Print(in Prefix,in Value,in Enabled,in ScreenDuration,in ScreenColor)"}
   Cast = RigVMDispatchNode(Value=@n1.triggerAxis) {template="DISPATCH_RigVMDispatch_CastEnumToInt(in Value,out Result)"}
   DISPATCH_RigVMDispatch_SelectInt32 = RigVMDispatchNode(
    Values=(0,0,0)
    Index=@Cast.Result
    Values.0=@RigVMFunction_MathQuaternionToEuler.Result.X
    Values.1=@RigVMFunction_MathQuaternionToEuler.Result.Y
    Values.2=@RigVMFunction_MathQuaternionToEuler.Result.Z
   ) {template="DISPATCH_RigVMDispatch_SelectInt32(in Index,in Values,out Result)"}
   Remap = &RigVMFunction_MathDoubleRemap(
    bClamp=False
    Value=@DISPATCH_RigVMDispatch_SelectInt32.Result
    TargetMinimum=@n1.outputMinMax.X
    TargetMaximum=@n1.outputMinMax.Y
    SourceMinimum=@n1.rotationMinMax.X
    SourceMaximum=@n1.rotationMinMax.Y
   )
   Absolute = &RigVMFunction_MathDoubleAbs(Value=@Remap.Result)
   Multiply = &RigVMFunction_MathDoubleMul(
    A=&RigVMFunction_MathDoubleClamp(Value=RigVMDispatchNode(True=@Absolute.Result, Condition=@n1.absoluteValue, False=@Remap.Result) {template="DISPATCH_RigVMDispatch_If(in Condition,in True,in False,out Result)"}.Result).Result
    B=@n1.premult
   )
   > RigVMFunctionReturnNode(outputQuat=@n2.Result, outputWeight=@Multiply.Result)
  ]}
 RigVMFunctionLibrary {nodes [
   DeltaRotation v4 = RigVMCollapseNode(
    ExecuteContext=()
    Rotation Order=ZYX
    child=(Type=Bone,Name="None")
    outputMinMax=(X=0,Y=1)
    parent=(Type=Bone,Name="None")
    premult=1
    rotationMinMax=(X=0,Y=0)
    twistAxis=(X=1,Y=0,Z=0)
   )
   DeltaRotation v4_1 = RigVMCollapseNode(
    ExecuteContext=()
    Rotation Order=ZYX
    child=(Type=Bone,Name="None")
    outputMinMax=(X=0,Y=1)
    parent=(Type=Bone,Name="None")
    premult=1
    rotationMinMax=(X=0,Y=0)
    twistAxis=(X=1,Y=0,Z=0)
   )
   Neck Collar Constraints = RigVMCollapseNode(ExecuteContext=())
   Neck Collar Get Variables = RigVMCollapseNode(ExecuteContext=())
   skirt = RigVMCollapseNode(ExecuteContext=())
  ]}
 RigVMFunction_Sequence_1_ContainedGraph {nodes [
   n1 = RigVMFunctionEntryNode()
   RigVMFunction_Sequence = &RigVMFunction_Sequence(ExecuteContext=@n1.ExecuteContext)
   n2 = &RigVMFunction_Sequence(ExecuteContext=@RigVMFunction_Sequence.B)
   RigVMFunctionReturnNode(A=@RigVMFunction_Sequence.A, B=@n2.A, C=@n2.B)
  ]}
 RigVMFunction_Sequence_1_ContainedGraph {nodes [
   n1 = RigVMFunctionEntryNode()
   RigVMFunction_Sequence = &RigVMFunction_Sequence(ExecuteContext=@n1.ExecuteContext)
   n2 = &RigVMFunction_Sequence(ExecuteContext=@RigVMFunction_Sequence.B)
   n3 = &RigVMFunction_Sequence(ExecuteContext=@n2.B)
   RigVMFunctionReturnNode(A=@RigVMFunction_Sequence.A, B=@n2.A, C=@n3.A, D=@n3.B)
  ]}
 RigVMFunction_Sequence_2_ContainedGraph {nodes [
   n1 = RigVMFunctionEntryNode()
   n2 = &RigVMFunction_Sequence(ExecuteContext=@n1.ExecuteContext)
   n3 = &RigVMFunction_Sequence(ExecuteContext=@n2.B)
   n4 = &RigVMFunction_Sequence(ExecuteContext=@n3.B)
   RigVMFunctionReturnNode(A=@n2.A, B=@n3.A, C=@n4.A, D=@n4.B)
  ]}
 RigVMModel {nodes [
   RigUnit_BeginExecution = &RigUnit_BeginExecution()
   FN_DeformRig_FNC = RigVMFunctionReferenceNode(
    DeformRig_ Curve="(deformRig_l_upperarm_up_alpha_72_#1=False,deformRig_r_upperarm_up_alpha_73_#2=False,deformRig_l_upperarm_fr_alpha_74_#3=False,deformRig_r_upperarm_fr_alpha_75_CB51B8 ...(truncated)"
    DeformRig_Bool="(upperarm_1_#4=True,elbow_3_#5=True,wrist_7_#6=True,groin_8_#7=True,glute_17_#8=True,knee_9_40C408324CA261 ...(truncated)"
    DeformRig_Debug=(upperarm_debug_18_#4=False,wrist_debug_20_#6=False,lat_debug_26_#9=False,groinGlute_debug_27_#7=False)
    DeformRig_Value="(upperarm_up_18_#10=(Rotation=(X=0.0,Y=-0.0,Z=0.0,W=1.0),Translation=(X=0.0,Y=0.0,Z=8.0),Scale3D=(X=0.70,Y=1.0,Z=1.0)),upperarm_fwd_fr_19_#11=(Rotation=(X=0.0,Y=-0.0,Z=0.0,W=1.0),Translation=(X=0.0, ...(truncated)"
    ExecuteContext=@RigUnit_BeginExecution.ExecutePin
   )
   > skirt = RigVMFunctionReferenceNode()
   > RigVMFunction_Sequence = RigVMAggregateNode()
   Neck Collar Constraints = RigVMFunctionReferenceNode(ExecuteContext=@RigVMFunction_Sequence.A)
   > Neck Collar Get Variables = RigVMFunctionReferenceNode()
   > &RigUnit_ModifyTransforms(
    ItemToModify="((Item=(Type=Bone,Name=\"hood_side_l\"),Transform=(Rotation=(X=0,Y=0,Z=0.3827,W=0.9239),Translation=(X=0,Y=5,Z=0),Scale3D=(X=1,Y=1,Z=1))),(Item=(Type=Bone,Name=\"hood_front_l\"),Transform=(Rotation=(X=0,Y=0,Z=0.3827,W=0.9239),Translation=(X=0,Y=0,Z=0),Scale3D= ...(truncated)"
    WeightMinimum=0
    Weight=RigVMVariableNode() {variable=Head side l}.Value
   )
   > &RigUnit_ModifyTransforms(
    ItemToModify="((Item=(Type=Bone,Name=\"hood_side_r\"),Transform=(Rotation=(X=0.099,Y=-0.2391,Z=0.3696,W=0.8924),Translation=(X=0,Y=-5,Z=0),Scale3D=(X=1,Y=1,Z=1))),(Item=(Type=Bone,Name=\"hood_front_r\"),Transform=(Rotation=(X=0,Y=0,Z=0.3827,W=0.9239),Translation=(X=0,Y=0,Z= ...(truncated)"
    WeightMinimum=0
    Weight=RigVMVariableNode() {variable=Head side r}.Value
   )
   > &RigUnit_ModifyTransforms(
    ItemToModify="((Item=(Type=Bone,Name=\"hood_back_c\"),Transform=(Rotation=(X=0,Y=0,Z=-1,W=0),Translation=(X=0,Y=-6,Z=9),Scale3D=(X=0.2,Y=1,Z=1))),(Item=(Type=Bone,Name=\"hood_back_side_r\"),Transform=(Rotation=(X=0,Y=0,Z=0.6428,W=0.766),Translation=(X=0,Y=-7,Z=0),Scale3D=(X ...(truncated)"
    WeightMinimum=0
    Weight=RigVMVariableNode() {variable=Head back c}.Value
   )
   n1 = &RigUnit_GetRelativeTransformForItem(Child=(Type=Bone,Name="clavicle_l"), Parent=(Type=Control,Name="ref_clavicle_l_ctrl"), bChildInitial=False, bParentInitial=False)
   RigVMVariableNode(
    Value=&RigVMFunction_MathQuaternionToEuler(RotationOrder=ZYX, Value=@n1.RelativeTransform.Rotation).Result
    ExecuteContext=@RigVMFunction_Sequence.B
   ) {variable=Variable rot l}
   n2 = &RigUnit_GetRelativeTransformForItem(Child=(Type=Bone,Name="clavicle_r"), Parent=(Type=Control,Name="ref_clavicle_r_ctrl"), bChildInitial=False, bParentInitial=False)
   > RigVMVariableNode(Value=&RigVMFunction_MathQuaternionToEuler(RotationOrder=ZYX, Value=@n2.RelativeTransform.Rotation).Result) {variable=Variable rot r}
   n3 = &RigUnit_SphericalPoseReader(
    ActiveRegionScaleFactors=(PositiveWidth=1,NegativeWidth=2,PositiveHeight=2,NegativeHeight=2)
    ActiveRegionSize=0.2
    Debug=(bDrawDebug=True,bDraw2D=False,bDrawLocalAxes=False,DebugScale=25,DebugSegments=20,DebugThickness=0.25)
    DriverAxis=(X=1,Y=0,Z=0)
    DriverItem=(Type=Bone,Name="spine_05")
    FalloffRegionScaleFactors=(PositiveWidth=1,NegativeWidth=0,PositiveHeight=0,NegativeHeight=0)
    FalloffSize=0.4
    FlipHeightScaling=False
    FlipWidthScaling=False
    OptionalParentItem=(Type=Bone,Name="spine_01")
    RotationOffset=(X=90,Y=0,Z=40)
    ExecutePin=@RigVMFunction_Sequence.C
   )
   Subtract = &RigVMFunction_MathDoubleSub(B=@n3.OutputParam, A=RigVMVariableNode() {variable=leg_up_l}.Value)
   > RigVMVariableNode(Value=@Subtract.Result) {variable=leg_up_l}
   > RigVMVariableNode(Value=&RigVMFunction_MathDoubleSub(B=@n3.OutputParam, A=RigVMVariableNode() {variable=leg_up_r}.Value).Result) {variable=leg_up_r}
   &RigUnit_ModifyTransforms(
    ExecutePin=()
    ItemToModify=((Item=(Type=Bone,Name="dyn_lace_c_01_l"),Transform=(Rotation=(X=0,Y=0,Z=0,W=1),Translation=(X=0,Y=0,Z=0),Scale3D=(X=1,Y=1,Z=1))))
    Weight=1
    WeightMinimum=0
   )
  ]}
 RigVMModel Construction Event Graph {nodes [
   n1 = &RigUnit_PrepareForExecution()
   &RigUnit_HierarchyImportFromSkeleton(NameSpace=None, bIncludeCurves=False, bIncludeMeshSockets=False, bIncludeVirtualBones=True, ExecutePin=@n1.ExecutePin)
   > Control Stack at Item = RigVMFunctionReferenceNode(
    Add Bottom Null=False
    Add Null=true
    Add Secondary Control=False
    Control Suffix=_ctrl
    Item=(Name="head",Type=Bone)
    Name=ref_head
    Null Suffix=_null
    Parent=(Name="neck_02",Type=Bone)
    Secondary Settings="(Shape=(bVisible=True,Name=\"Default\",Color=(R=1,G=0,B=0,A=1),Transform=(Rotation=(X=0,Y=0,Z=0,W=1),Translation=(X=0,Y=0,Z=0),Scale3D=(X=1,Y=1,Z=1))),Proxy=(bIsProxy=False,ShapeVisibility=BasedOnSelection),DisplayName=\"None\",InitialSpace=LocalSpace,bUsePref ...(truncated)"
    Settings="(Shape=(bVisible=True,Name=\"Pyramid_Thin\",Color=(R=1,G=0,B=0,A=1),Transform=(Rotation=(X=0,Y=0,Z=0,W=1),Translation=(X=0,Y=0,Z=0),Scale3D=(X=1,Y=1,Z=1))),Proxy=(bIsProxy=False,ShapeVisibility=BasedOnSelection),DisplayName=\"None\",InitialSpace=LocalSpace,bUs ...(truncated)"
   )
   > Control Stack at Item_1 = RigVMFunctionReferenceNode(
    Add Bottom Null=False
    Add Null=true
    Add Secondary Control=False
    Control Suffix=_ctrl
    Item=(Name="neck_02",Type=Bone)
    Name=ref_neck
    Null Suffix=_null
    Parent=(Name="spine_05",Type=Bone)
    Secondary Settings="(Shape=(bVisible=True,Name=\"Default\",Color=(R=1,G=0,B=0,A=1),Transform=(Rotation=(X=0,Y=0,Z=0,W=1),Translation=(X=0,Y=0,Z=0),Scale3D=(X=1,Y=1,Z=1))),Proxy=(bIsProxy=False,ShapeVisibility=BasedOnSelection),DisplayName=\"None\",InitialSpace=LocalSpace,bUsePref ...(truncated)"
    Settings="(Shape=(bVisible=True,Name=\"Pyramid_Thin\",Color=(R=1,G=0,B=0,A=1),Transform=(Rotation=(X=0,Y=0,Z=0,W=1),Translation=(X=0,Y=0,Z=0),Scale3D=(X=1,Y=1,Z=1))),Proxy=(bIsProxy=False,ShapeVisibility=BasedOnSelection),DisplayName=\"None\",InitialSpace=LocalSpace,bUs ...(truncated)"
   )
   > Control Stack at Item_2 = RigVMFunctionReferenceNode(
    Add Bottom Null=False
    Add Null=true
    Add Secondary Control=False
    Control Suffix=_ctrl
    Item=(Name="clavicle_l",Type=Bone)
    Name=ref_clavicle_l
    Null Suffix=_null
    Parent=(Name="spine_05",Type=Bone)
    Secondary Settings="(Shape=(bVisible=True,Name=\"Default\",Color=(R=1,G=0,B=0,A=1),Transform=(Rotation=(X=0,Y=0,Z=0,W=1),Translation=(X=0,Y=0,Z=0),Scale3D=(X=1,Y=1,Z=1))),Proxy=(bIsProxy=False,ShapeVisibility=BasedOnSelection),DisplayName=\"None\",InitialSpace=LocalSpace,bUsePref ...(truncated)"
    Settings="(Shape=(bVisible=True,Name=\"Pyramid_Thin\",Color=(R=1,G=0,B=0,A=1),Transform=(Rotation=(X=0,Y=0,Z=0,W=1),Translation=(X=0,Y=0,Z=0),Scale3D=(X=1,Y=1,Z=1))),Proxy=(bIsProxy=False,ShapeVisibility=BasedOnSelection),DisplayName=\"None\",InitialSpace=LocalSpace,bUs ...(truncated)"
   )
   > Control Stack at Item_3 = RigVMFunctionReferenceNode(
    Add Bottom Null=False
    Add Null=true
    Add Secondary Control=False
    Control Suffix=_ctrl
    Item=(Name="clavicle_r",Type=Bone)
    Name=ref_clavicle_r
    Null Suffix=_null
    Parent=(Name="spine_05",Type=Bone)
    Secondary Settings="(Shape=(bVisible=True,Name=\"Default\",Color=(R=1,G=0,B=0,A=1),Transform=(Rotation=(X=0,Y=0,Z=0,W=1),Translation=(X=0,Y=0,Z=0),Scale3D=(X=1,Y=1,Z=1))),Proxy=(bIsProxy=False,ShapeVisibility=BasedOnSelection),DisplayName=\"None\",InitialSpace=LocalSpace,bUsePref ...(truncated)"
    Settings="(Shape=(bVisible=True,Name=\"Pyramid_Thin\",Color=(R=1,G=0,B=0,A=1),Transform=(Rotation=(X=0,Y=0,Z=0,W=1),Translation=(X=0,Y=0,Z=0),Scale3D=(X=1,Y=1,Z=1))),Proxy=(bIsProxy=False,ShapeVisibility=BasedOnSelection),DisplayName=\"None\",InitialSpace=LocalSpace,bUs ...(truncated)"
   )
   > &RigUnit_HierarchyAddNull(
    Name=thigh_l_null
    Parent=(Type=Bone,Name="pelvis")
    Space=GlobalSpace
    Transform=&RigUnit_GetTransform(Item=(Type=Bone,Name="thigh_l"), bInitial=false).Transform
   )
   > &RigUnit_HierarchyAddNull(
    Name=thigh_r_null
    Parent=(Type=Bone,Name="pelvis")
    Space=GlobalSpace
    Transform=&RigUnit_GetTransform(Item=(Type=Bone,Name="thigh_r"), bInitial=false).Transform
   )
   > &RigUnit_HierarchyAddNull(
    Name=frontSkirt_01_null
    Parent=(Type=Bone,Name="dyn_frontSkirt_01")
    Space=LocalSpace
    Transform=(Rotation=(X=0,Y=0,Z=0,W=1),Translation=(X=0,Y=-10,Z=0),Scale3D=(X=1,Y=1,Z=1))
   )
   > &RigUnit_HierarchyAddNull(
    Name=frontSkirt_02_null
    Parent=(Type=Bone,Name="dyn_frontSkirt_02")
    Space=LocalSpace
    Transform=(Rotation=(X=0,Y=0,Z=0,W=1),Translation=(X=1,Y=-10,Z=0),Scale3D=(X=1,Y=1,Z=1))
   )
   > &RigUnit_HierarchyAddNull(
    Name=frontSkirt_01_l_null
    Parent=(Type=Bone,Name="dyn_frontSkirt_01_l")
    Space=LocalSpace
    Transform=(Rotation=(X=0,Y=0,Z=0,W=1),Translation=(X=0,Y=2,Z=0.5),Scale3D=(X=1,Y=1,Z=1))
   )
   > &RigUnit_HierarchyAddNull(
    Name=frontSkirt_02_l_null
    Parent=(Type=Bone,Name="dyn_frontSkirt_02_l")
    Space=LocalSpace
    Transform=(Rotation=(X=0,Y=0,Z=0,W=1),Translation=(X=0,Y=1,Z=1),Scale3D=(X=1,Y=1,Z=1))
   )
   > &RigUnit_HierarchyAddNull(
    Name=calf_l_null
    Parent=(Type=Bone,Name="calf_l")
    Space=LocalSpace
    Transform=(Rotation=(X=0,Y=0,Z=0,W=1),Translation=(X=0,Y=12.5,Z=0),Scale3D=(X=1,Y=1,Z=1))
   )
   addFKCtrl = RigVMFunctionReferenceNode(
    Custom_FKCtrl=()
    ExecuteContext=()
    Options_FKCtrl="(pelvis_1_#12=true,spine_3_#13=true,clavicle_5_#14=true,arm_7_#15=true,leg_9_#16=true,head_11_46DC15EF465CD2979 ...(truncated)"
    Shape_Settings="(pelvis_2_#17=(bVisible=True,Name=\"Hexagon_Thin\",Color=(R=0.0117,G=0.8281,B=1,A=1),Transform=(Rotation=(X=0,Y=-0.7071,Z=0,W=0.7071),Translation=(X=2,Y=0,Z=0),Scale3D=(X=5,Y=5,Z=7))),spine_4_#18=(bVi ...(truncated)"
   )
  ]}
]
variables [
 Head back c {type=double, default=0}
 Head side l {type=double, default=0}
 Head side r {type=double, default=0}
 Variable rot l {type=FVector, default=(X=0,Y=0,Z=0)}
 Variable rot r {type=FVector, default=(X=0,Y=0,Z=0)}
 is crouching {type=bool, default=False}
 leg up crouching {type=double, default=0}
 leg up crouching_r {type=double, default=0}
 leg up out crouching {type=double, default=0}
 leg up out crouching_r {type=double, default=0}
 leg_back_l {type=double, default=0}
 leg_out_l {type=double, default=0}
 leg_out_r {type=double, default=0}
 leg_up_l {type=double, default=0}
 leg_up_r {type=double, default=0}
]
