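# Physics asset for the front skirt and lace strands of the OvenDrastic male medium body, built entirely from capsule bodies.
# Pelvis and thigh capsules provide body volumes, with three two-bone front skirt chains and five lace chains branching from the pelvis.
# Skirt, thigh and most base-lace joints swing and twist freely, while dynamic lace links have tight angular limits.
# Center and left lower skirt bodies stack several angled capsules to approximate the panel shape.
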PhysicsAsset /BRCosmetics/Characters/Player/Male/Medium/Bodies/M_MED_OvenDrastic/Meshes/M_MED_OvenDrastic_SkirtPhysics
bodies {
 base_lace_01_r [{shapeType=Capsule, length=0.505, radius=0.505, shapeName=base_lace_01_r_capsule, rotation=<rot 0 -90 0>}]
 base_lace_a_01_l [{shapeType=Capsule, length=0.505, radius=0.505, shapeName=base_lace_a_01_l_capsule, rotation=<rot 0 -90 0>}]
 base_lace_b_01_l [{shapeType=Capsule, length=0.505, radius=0.505, shapeName=base_lace_b_01_l_capsule, rotation=<rot 0 -90 0>}]
 base_lace_b_02_l [{shapeType=Capsule, length=0.505, radius=0.505, shapeName=base_lace_b_02_l_capsule, rotation=<rot 0 -90 0>}]
 base_lace_c_01_l [{shapeType=Capsule, length=0.505, radius=0.505, shapeName=base_lace_c_01_l_capsule, rotation=<rot 0 -90 0>}]
 base_lace_c_02_l [{shapeType=Capsule, length=0.505, radius=0.505, shapeName=base_lace_c_02_l_capsule, rotation=<rot 0 -90 0>}]
 base_lace_d_01_l [{shapeType=Capsule, length=0.505, radius=0.505, shapeName=base_lace_d_01_l_capsule, rotation=<rot 0 -90 0>}]
 dyn_frontSkirt_01 [
  {shapeType=Capsule, length=9.7, radius=4.5, shapeName=dyn_frontSkirt_01_capsule, center=<2 0 -6.8571>, rotation=<rot 0 0 -90>}
  {
   shapeType=Capsule
   length=9.7
   radius=2.5
   shapeName=dyn_frontSkirt_01_capsule
   center=<1.6301 -9.9521 -6.8571>
   rotation=<rot 0.0001 0 -118.4326>
  }
 ]
 dyn_frontSkirt_01_l [{shapeType=Capsule, length=9.7, radius=4.5, shapeName=dyn_frontSkirt_01_l_capsule, center=<2 0 -6.8571>, rotation=<rot 0 0 -96>}]
 dyn_frontSkirt_01_r [{shapeType=Capsule, length=9.7, radius=4.5, shapeName=dyn_frontSkirt_01_r_capsule, center=<2 0 -6.8571>, rotation=<rot 0 0 -84>}]
 dyn_frontSkirt_02 [
  {shapeType=Capsule, length=11, radius=5.5, shapeName=dyn_frontSkirt_02_capsule, center=<-0.1842 0 -4.5737>, rotation=<rot 0 0 -90>}
  {
   shapeType=Capsule
   length=14
   radius=3.3
   shapeName=dyn_frontSkirt_02_capsule
   center=<-3.118 -9.6452 -5.1154>
   rotation=<rot -3.4795 -9.6728 -124.7789>
  }
  {
   shapeType=Capsule
   length=14.3
   radius=3.3
   shapeName=dyn_frontSkirt_02_capsule
   center=<0.4059 -7.0862 -3.418>
   rotation=<rot -3.4788 -9.6728 -119.0574>
  }
  {
   shapeType=Capsule
   length=12.2
   radius=2.6
   shapeName=dyn_frontSkirt_02_capsule
   center=<1.7019 -8.0201 -3.4985>
   rotation=<rot -3.4775 -9.6727 -107.5092>
  }
  {
   shapeType=Capsule
   length=15.4
   radius=4.5
   shapeName=dyn_frontSkirt_02_capsule
   center=<2.073 -5.8777 -4.1724>
   rotation=<rot -2.7975 -6.0723 -117.4226>
  }
  {
   shapeType=Capsule
   length=15
   radius=2
   shapeName=dyn_frontSkirt_02_capsule
   center=<0.9186 -7.182 -1.5136>
   rotation=<rot -2.7974 -6.0723 -120.2831>
  }
 ]
 dyn_frontSkirt_02_l [
  {
   shapeType=Capsule
   length=5
   radius=1.7
   shapeName=dyn_frontSkirt_02_l_capsule
   center=<-2.0588 -2.3057 -6.0072>
   rotation=<rot 0.0008 0.0004 -80.9624>
  }
  {
   shapeType=Capsule
   length=5
   radius=1.7
   shapeName=dyn_frontSkirt_02_l_capsule
   center=<0.8344 -2.3058 -6.0073>
   rotation=<rot 0.0008 0.0004 -80.9624>
  }
  {
   shapeType=Capsule
   length=5
   radius=1.7
   shapeName=dyn_frontSkirt_02_l_capsule
   center=<3.758 -2.3059 -6.0074>
   rotation=<rot 0.0008 0.0004 -80.9624>
  }
  {
   shapeType=Capsule
   length=11.1
   radius=1.7
   shapeName=dyn_frontSkirt_02_l_capsule
   center=<-0.9958 2.9423 -6.7334>
   rotation=<rot 0.0013 0.0007 -54.3609>
  }
  {
   shapeType=Capsule
   length=11.1
   radius=1.7
   shapeName=dyn_frontSkirt_02_l_capsule
   center=<2.0536 2.9663 -6.7428>
   rotation=<rot 0.0013 0.0007 -54.3609>
  }
 ]
 dyn_frontSkirt_02_r [{shapeType=Capsule, length=9.7, radius=5.5, shapeName=dyn_frontSkirt_02_r_capsule, center=<2.5163 0 -6.1899>, rotation=<rot 0 0 -90>}]
 dyn_lace_01_r [{shapeType=Capsule, length=2, radius=2, shapeName=dyn_lace_01_r_capsule, center=<5 0 1>, rotation=<rot -90 0 0>}]
 dyn_lace_02_r [{shapeType=Capsule, length=2, radius=3, shapeName=dyn_lace_02_r_capsule, center=<5 0 1>, rotation=<rot -90 0 0>}]
 dyn_lace_03_r [{shapeType=Capsule, length=2, radius=3, shapeName=dyn_lace_03_r_capsule, center=<5 0 1>, rotation=<rot -90 0 0>}]
 dyn_lace_04_r [{shapeType=Capsule, length=2, radius=4, shapeName=dyn_lace_04_r_capsule, center=<5 0 1>, rotation=<rot -90 0 0>}]
 dyn_lace_a_01_l [{shapeType=Capsule, length=2, radius=2, shapeName=dyn_lace_a_01_l_capsule, center=<2.7006 0 -3.1301>, rotation=<rot -90 0 0>}]
 dyn_lace_a_02_l [{shapeType=Capsule, length=2, radius=2, shapeName=dyn_lace_a_02_l_capsule, center=<2.7006 0 -3.1301>, rotation=<rot -90 0 0>}]
 dyn_lace_a_03_l [{shapeType=Capsule, length=2, radius=2, shapeName=dyn_lace_a_03_l_capsule, center=<2.7006 0 -3.1301>, rotation=<rot -90 0 0>}]
 dyn_lace_b_01_l [{shapeType=Capsule, length=2, radius=2, shapeName=dyn_lace_b_01_l_capsule, center=<2.1583 0 -2.1704>, rotation=<rot -90 0 0>}]
 dyn_lace_b_02_l [{shapeType=Capsule, length=2, radius=2, shapeName=dyn_lace_b_02_l_capsule, center=<1.2947 0 -2.5394>, rotation=<rot -90 0 0>}]
 dyn_lace_b_03_l [{shapeType=Capsule, length=4.5, radius=2, shapeName=dyn_lace_b_03_l_capsule, center=<4.8634 0 -0.7292>, rotation=<rot -90 0 0>}]
 dyn_lace_b_04_l [{shapeType=Capsule, length=8, radius=2, shapeName=dyn_lace_b_04_l_capsule, center=<5.6871 0 -0.2644>, rotation=<rot -86.1016 0.0001 -0.0001>}]
 dyn_lace_c_01_l [{shapeType=Capsule, length=5, radius=2, shapeName=dyn_lace_c_01_l_capsule, center=<4 0 -1.2441>, rotation=<rot -90 0 0>}]
 dyn_lace_c_02_l [{shapeType=Capsule, length=5, radius=2, shapeName=dyn_lace_c_02_l_capsule, center=<4 0 -1.5>, rotation=<rot -90 0 0>}]
 dyn_lace_c_03_l [{shapeType=Capsule, length=5, radius=2.3, shapeName=dyn_lace_c_03_l_capsule, center=<3.0302 0 -1.35>, rotation=<rot -90 0 0>}]
 dyn_lace_d_01_l [{shapeType=Capsule, length=1, radius=3, shapeName=dyn_lace_d_01_l_capsule, center=<3 0 -2.5>, rotation=<rot 90 0 0>}]
 dyn_lace_d_02_l [{shapeType=Capsule, length=1, radius=3, shapeName=dyn_lace_d_02_l_capsule, center=<3 0 -2.5>, rotation=<rot 90 0 0>}]
 dyn_lace_d_03_l [{shapeType=Capsule, length=1, radius=3, shapeName=dyn_lace_d_03_l_capsule, center=<2.965 0 -2.4787>, rotation=<rot 87.9524 -0.0021 -0.0019>}]
 pelvis [{shapeType=Capsule, length=19.5, radius=14.5, shapeName=pelvis_capsule, center=<-0.3324 -4.4784 0.2708>, rotation=<rot 87.4463 0 2.5609>}]
 thigh_l [
  {shapeType=Capsule, length=35, radius=9, shapeName=thigh_l_capsule, center=<-13.5591 -2.3009 0.3306>, rotation=<rot -10.611 -87.3772 87.5552>}
  {shapeType=Capsule, length=10, radius=2.2, shapeName=thigh_l_capsule, center=<-19.4111 3.7307 -6.3973>, rotation=<rot 61.5987 -90.0008 0>}
  {shapeType=Capsule, length=8.9, radius=2.6, shapeName=thigh_l_capsule, center=<-8.9878 3.0246 -4.7267>, rotation=<rot 73.6046 -90.001 0>}
 ]
 thigh_r [{
   shapeType=Capsule
   length=25
   radius=10.9102
   shapeName=thigh_r_capsule
   center=<17.8716 0.6364 0.4953>
   rotation=<rot -10.7534 -91.3766 88.2969>
  }]
}
constraints [
 {
  bone1Name=thigh_l
  bone2Name=pelvis
  swing1LimitDegrees=45
  swing1Motion=Free
  swing2LimitDegrees=45
  swing2Motion=Free
  twistLimitDegrees=45
  twistMotion=Free
 }
 {
  bone1Name=thigh_r
  bone2Name=pelvis
  swing1LimitDegrees=45
  swing1Motion=Free
  swing2LimitDegrees=45
  swing2Motion=Free
  twistLimitDegrees=45
  twistMotion=Free
 }
 {
  bone1Name=base_lace_01_r
  bone2Name=pelvis
  swing1LimitDegrees=45
  swing1Motion=Free
  swing2LimitDegrees=45
  swing2Motion=Free
  twistLimitDegrees=45
  twistMotion=Free
 }
 {bone1Name=dyn_lace_01_r, bone2Name=base_lace_01_r, swing1LimitDegrees=5, swing2LimitDegrees=5, twistLimitDegrees=1}
 {bone1Name=dyn_lace_02_r, bone2Name=dyn_lace_01_r, swing1LimitDegrees=5, swing2LimitDegrees=5, twistLimitDegrees=1}
 {bone1Name=dyn_lace_03_r, bone2Name=dyn_lace_02_r, swing1LimitDegrees=5, swing2LimitDegrees=5, twistLimitDegrees=1}
 {bone1Name=dyn_lace_04_r, bone2Name=dyn_lace_03_r, swing1LimitDegrees=5, swing2LimitDegrees=5, twistLimitDegrees=1}
 {
  bone1Name=base_lace_a_01_l
  bone2Name=pelvis
  swing1LimitDegrees=45
  swing1Motion=Free
  swing2LimitDegrees=45
  swing2Motion=Free
  twistLimitDegrees=45
  twistMotion=Free
 }
 {bone1Name=dyn_lace_a_01_l, bone2Name=base_lace_a_01_l, swing1LimitDegrees=5, swing2LimitDegrees=5, twistLimitDegrees=1}
 {bone1Name=dyn_lace_a_02_l, bone2Name=dyn_lace_a_01_l, swing1LimitDegrees=5, swing2LimitDegrees=5, twistLimitDegrees=1}
 {bone1Name=dyn_lace_a_03_l, bone2Name=dyn_lace_a_02_l, swing1LimitDegrees=5, swing2LimitDegrees=5, twistLimitDegrees=1}
 {bone1Name=base_lace_b_01_l, bone2Name=pelvis, swing1LimitDegrees=25, swing2LimitDegrees=25, twistLimitDegrees=1}
 {bone1Name=dyn_lace_b_01_l, bone2Name=base_lace_b_01_l, swing1LimitDegrees=5, swing2LimitDegrees=5, twistLimitDegrees=1}
 {bone1Name=dyn_lace_b_02_l, bone2Name=base_lace_b_02_l, swing1LimitDegrees=5, swing2LimitDegrees=5, twistLimitDegrees=1}
 {
  bone1Name=base_lace_c_01_l
  bone2Name=pelvis
  swing1LimitDegrees=45
  swing1Motion=Free
  swing2LimitDegrees=45
  swing2Motion=Free
  twistLimitDegrees=45
  twistMotion=Free
 }
 {bone1Name=dyn_lace_c_01_l, bone2Name=base_lace_c_01_l, swing1LimitDegrees=5, swing2LimitDegrees=5, twistLimitDegrees=1}
 {bone1Name=dyn_lace_c_02_l, bone2Name=base_lace_c_02_l, swing1LimitDegrees=5, swing2LimitDegrees=20, twistLimitDegrees=1}
 {bone1Name=dyn_lace_c_03_l, bone2Name=dyn_lace_c_02_l, swing1LimitDegrees=5, swing2LimitDegrees=20, twistLimitDegrees=1}
 {
  bone1Name=base_lace_d_01_l
  bone2Name=pelvis
  swing1LimitDegrees=45
  swing1Motion=Free
  swing2LimitDegrees=45
  swing2Motion=Free
  twistLimitDegrees=45
  twistMotion=Free
 }
 {bone1Name=dyn_lace_d_01_l, bone2Name=base_lace_d_01_l, swing1LimitDegrees=3, swing2LimitDegrees=3, twistLimitDegrees=1}
 {bone1Name=dyn_lace_d_02_l, bone2Name=dyn_lace_d_01_l, swing1LimitDegrees=3, swing2LimitDegrees=3, twistLimitDegrees=1}
 {bone1Name=dyn_lace_d_03_l, bone2Name=dyn_lace_d_02_l, swing1LimitDegrees=3, swing2LimitDegrees=3, twistLimitDegrees=1}
 {
  bone1Name=base_lace_b_02_l
  bone2Name=dyn_lace_b_01_l
  swing1LimitDegrees=45
  swing1Motion=Free
  swing2LimitDegrees=45
  swing2Motion=Free
  twistLimitDegrees=45
  twistMotion=Free
 }
 {
  bone1Name=base_lace_c_02_l
  bone2Name=dyn_lace_c_01_l
  swing1LimitDegrees=45
  swing1Motion=Free
  swing2LimitDegrees=45
  swing2Motion=Free
  twistLimitDegrees=45
  twistMotion=Free
 }
 {
  bone1Name=dyn_frontSkirt_02
  bone2Name=dyn_frontSkirt_01
  swing1LimitDegrees=45
  swing1Motion=Free
  swing2LimitDegrees=45
  swing2Motion=Free
  twistLimitDegrees=45
  twistMotion=Free
 }
 {
  bone1Name=dyn_frontSkirt_02_l
  bone2Name=dyn_frontSkirt_01_l
  swing1LimitDegrees=45
  swing1Motion=Free
  swing2LimitDegrees=45
  swing2Motion=Free
  twistLimitDegrees=45
  twistMotion=Free
 }
 {
  bone1Name=dyn_frontSkirt_02_r
  bone2Name=dyn_frontSkirt_01_r
  swing1LimitDegrees=45
  swing1Motion=Free
  swing2LimitDegrees=45
  swing2Motion=Free
  twistLimitDegrees=45
  twistMotion=Free
 }
 {
  bone1Name=dyn_frontSkirt_01
  bone2Name=pelvis
  swing1LimitDegrees=45
  swing1Motion=Free
  swing2LimitDegrees=45
  swing2Motion=Free
  twistLimitDegrees=45
  twistMotion=Free
 }
 {
  bone1Name=dyn_frontSkirt_01_l
  bone2Name=pelvis
  swing1LimitDegrees=45
  swing1Motion=Free
  swing2LimitDegrees=45
  swing2Motion=Free
  twistLimitDegrees=45
  twistMotion=Free
 }
 {
  bone1Name=dyn_frontSkirt_01_r
  bone2Name=pelvis
  swing1LimitDegrees=45
  swing1Motion=Free
  swing2LimitDegrees=45
  swing2Motion=Free
  twistLimitDegrees=45
  twistMotion=Free
 }
 {bone1Name=dyn_lace_b_03_l, bone2Name=dyn_lace_b_02_l, swing1LimitDegrees=5, swing2LimitDegrees=5, twistLimitDegrees=1}
 {bone1Name=dyn_lace_b_04_l, bone2Name=dyn_lace_b_03_l, swing1LimitDegrees=5, swing2LimitDegrees=15, twistLimitDegrees=1}
]
